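# Animation Blueprint for the face accessory mesh part of the male medium BlockStack GenericHair3 character body.
# Its AnimGraph is a single Copy Pose From Mesh node feeding the output pose, so the part mirrors another mesh's pose.
# The source is the Head budgeted skeletal mesh variable, with use of the attached parent enabled.
# The Event Graph holds only an unconnected update event, and nothing here assigns Head.
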AnimBlueprint /BRCosmetics/Characters/Player/Male/Medium/Bodies/M_MED_BlockStack_GenericHair3/Meshes/Parts/M_MED_BlockStack_GenericHair3_FaceAcc_AnimBP
target_skeleton=&/M_MED_BlockStack_GenericHair3_FaceAcc_Skeleton
graphs [
 AnimGraph {nodes [Output Pose(Result=Copy Pose From Mesh(SourceMeshComponent=Get Head() -> (Head:Skeletal Mesh Component Budgeted Object Reference)) {bUseAttachedParent=true}) {layerGroup=None}]}
 EventGraph {nodes [Event BlueprintUpdateAnimation()]}
]
variables [Head {type=SkeletalMeshComponentBudgeted}]
